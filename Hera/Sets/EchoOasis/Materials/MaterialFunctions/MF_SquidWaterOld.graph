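# Water material function for Echo Oasis that paints animated "squid water" rings and a spiral around a world-space pivot.
# Water depth, scrolling Perlin noise and a center-hole mask form sine rings, blended into rotating polar stripes near the pivot.
# This pattern drives scattering color and radial normals, with a mask from two hard-coded world-space segments limiting the effect.
# Static switches gate all outputs except Normal and select a mobile palette; many leftover nodes are unconnected.
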MaterialFunction /Hera/Sets/EchoOasis/Materials/MaterialFunctions/MF_SquidWaterOld
graph {
 properties {bExposeToLibrary=true}
 nodes [
  n1 = DepthFade(FadeDistance=ScalarParameter() {defaultValue=1000, parameterName=DepthFadeMaskDistance})
  Add(A=Multiply(A=ScalarParameter() {defaultValue=20, parameterName=DepthFadeMaskIntensity}, B=@n1))
  n2 = ScalarParameter() {defaultValue=1, parameterName=Blend factor}
  n3 = Divide(
   A=MaterialFunctionCall!(Position) {materialFunction=/Water/Materials/Functions/WaterBodyData}.Water Depth
   B=ScalarParameter() {defaultValue=400, parameterName=Waterbody Depth Scale}
  ) {constB=100}
  n4 = Divide(A=WorldPosition().XY) {constB=2000}
  n5 = Constant2Vector() {g=-0.3, r=-1}
  n6 = Multiply(A=Multiply(A=Constant() {r=0.02}, B=Time()), B=@n5)
  n7 = TextureSample(UVs=BumpOffset(Coordinate=Add(A=@n4, B=@n6), Height=ScalarParameter() {defaultValue=-3, parameterName=H})) {samplerType=SAMPLERTYPE_Masks, texture=/Game/Characters/Player/Female/Medium/Bodies/F_MED_Neon_Tight_Suit/Materials/Soft_Perlin_output}
  n8 = ScalarParameter() {defaultValue=0.05, parameterName=WBD Noise Intensity}
  n9 = Constant() {r=1}
  n10 = LinearInterpolate(A=@n9, B=Add(A=Saturate(None=@n3), B=Multiply(A=@n7.R, B=@n8)), Alpha=@n2)
  n11 = ComponentMask(None=Subtract(A=WorldPosition().XYZ, B=VectorParameter() {parameterName=SpiralPivot, defaultValue=<rgba 60300 34200 250 1>}.RGB)) {g=true, r=true}
  n12 = Divide(A=@n11) {constB=5000}
  n13 = Add(A=@n12) {constB=0.5}
  n14 = Subtract(A=Saturate(None=@n13), B=Constant() {r=0.5})
  n15 = Multiply(A=@n14, B=@n14)
  n16 = Add(A=ComponentMask(None=@n15) {r=true}, B=ComponentMask(None=@n15) {g=true})
  n17 = Time()
  n18 = Constant()
  n19 = Multiply(A=Add(A=SquareRoot(None=@n16), B=Multiply(A=@n17, B=@n18))) {constB=2}
  n20 = Saturate(None=OneMinus(None=@n19))
  n21 = Multiply(A=ScalarParameter() {defaultValue=1, parameterName=CenterHoleSize}, B=OneMinus(None=@n20))
  n22 = MaterialFunctionCall(Alpha=Saturate(None=OneMinus(None=@n21))) {materialFunction=&SmoothStep}
  n23 = Multiply(A=ScalarParameter() {defaultValue=0.5, parameterName=CenterHoleIntensity}, B=@n22.Result)
  Add(A=@n10, B=@n23)
  n24 = Arctangent2!(Y=ComponentMask(None=@n14) {r=true}, X=ComponentMask(None=@n14) {g=true})
  n25 = Constant() {r=2}
  n26 = Frac(None=Divide(
    A=ComponentMask(None=@n24) {r=true}
    B=MaterialFunctionCall(Multiplier=@n25) {materialFunction=&Pi}.Result
   ))
  n27 = Multiply(A=Constant() {r=-0.3}, B=@n17)
  n28 = Add(A=@n19) {constB=-0.05}
  n29 = OneMinus(None=Power(Base=OneMinus(None=@n28)) {constExponent=1})
  n30 = Add(A=Multiply(A=Add(A=@n26, B=@n27)), B=Multiply(A=@n29) {constB=15})
  n31 = Frac(None=Multiply(A=Frac(None=@n30)))
  n32 = Multiply(A=@n31) {constB=2}
  Ceil(None=Add(A=@n32) {constB=-1}) {desc=mask}
  Constant4Vector() {constant=<rgba 1 0 0.0833 1>}
  Constant4Vector() {constant=<rgba 0.8 0.04 0.1033 1>}
  Constant4Vector() {constant=<rgba 0.3 0 1.5 1>}
  Constant4Vector() {constant=<rgba 11 0 0.9166 1>}
  Constant4Vector() {constant=<rgba 0.1667 0 1 1>}
  Constant() {r=1}
  Constant() {r=1}
  Constant() {r=400}
  Constant() {r=1000}
  Constant() {r=0.05}
  Constant() {r=0.03}
  Constant() {r=1}
  Constant() {r=0.02}
  Constant()
  Constant() {r=60}
  DepthFade() {fadeDistanceDefault=1000}
  Divide() {constB=1000}
  Frac(None=@n32) {desc=uv}
  Frac(None=@n13)
  n33 = FunctionInput() {bUsePreviewValueAsDefault=true, inputName=Scattering, previewValue=<1 1 1 1>}
  n34 = VectorParameter() {parameterName=Color A, defaultValue=<rgba 0.5 0 1 1>}
  n35 = VectorParameter() {parameterName=Color B, defaultValue=<rgba 0.6667 0 1 1>}
  n36 = ScalarParameter() {defaultValue=20, parameterName=Rings}
  n37 = MaterialFunctionCall(In=Power(Base=@n22.Result) {constExponent=1}, Contrast=Constant() {r=0.2}) {materialFunction=&CheapContrast}
  n38 = ScalarParameter() {defaultValue=1, parameterName=Blend factor centerHole}
  n39 = Divide(A=LinearInterpolate(A=@n10, B=@n23, Alpha=Multiply(A=@n37.Result, B=@n38)) {constAlpha=0.9}) {constB=5}
  n40 = Constant() {r=0.03}
  n41 = Time()
  n42 = Sine(None=Add(A=Multiply(A=@n36, B=@n39), B=Multiply(A=@n40, B=@n41)))
  n43 = WorldPosition()
  n44 = Multiply(A=Constant() {r=0.05}, B=Time())
  n45 = TextureSample(UVs=Add(A=Divide(A=@n43.XY) {constB=5000}, B=Multiply(A=@n44, B=@n5))) {samplerType=SAMPLERTYPE_Masks, texture=&Soft_Perlin_output}
  n46 = LinearInterpolate(A=Sine(None=@n42), B=Cosine(None=@n42), Alpha=@n45.R)
  n47 = Saturate(None=@n46)
  n48 = Sine(None=@n31)
  n49 = OneMinus(None=Multiply(A=ScalarParameter() {defaultValue=1, parameterName=Stripes gradient offset}, B=@n48))
  n50 = OneMinus(None=Power(Base=@n20) {constExponent=22})
  n51 = MaterialFunctionCall(Alpha=Multiply(A=Multiply(A=@n49, B=@n50)) {constB=0.5}) {materialFunction=&SmoothStep}
  n52 = LinearInterpolate(A=@n47, B=@n51.Result, Alpha=Power(Base=@n20) {constExponent=5})
  n53 = Power(Base=OneMinus(None=@n52)) {constExponent=5}
  n54 = Saturate(None=Add(A=OneMinus(None=@n53)) {constB=-0.075})
  n55 = Constant() {r=0.5}
  n56 = Sine(None=Power(Base=MaterialFunctionCall(In=@n7.R, Contrast=@n55) {materialFunction=&CheapContrast}.Result))
  n57 = Frac(None=Add(A=Multiply(A=@n56) {constB=0.08}, B=@n54))
  n58 = Distance(A=@n57, B=Constant() {r=0.6})
  n59 = Multiply(A=Power(Base=Multiply(A=@n58))) {constB=2}
  n60 = Multiply(A=@n59, B=@n54)
  n61 = LinearInterpolate(
   A=LinearInterpolate(A=@n34.RGB, B=@n35.RGB, Alpha=@n54)
   B=VectorParameter() {parameterName=Color B2, defaultValue=<rgba 0 0.6667 1 1>}.RGB
   Alpha=@n60
  )
  n62 = MaterialFunctionCall() {materialFunction=&WaterBodyData}
  n63 = WorldPosition()
  n64 = Constant3Vector() {desc=A Coord, constant=<rgba 65296 35556 150 0>}
  n65 = Subtract(A=Constant3Vector() {desc=B Coord, constant=<rgba 73800 33476 150 0>}, B=@n64) {desc=AB}
  n66 = Divide(A=DotProduct(A=Subtract(A=@n63.XYZ, B=@n64) {desc=AP}, B=@n65), B=DotProduct(A=@n65, B=@n65))
  n67 = WorldPosition()
  n68 = Constant3Vector() {desc=B Coord, constant=<rgba 58296 33732 300 0>}
  n69 = Subtract(A=Constant3Vector() {desc=A Coord, constant=<rgba 57488 33144 300 0>}, B=@n68) {desc=AB}
  n70 = Divide(A=DotProduct(A=Subtract(A=@n67.XYZ, B=@n68) {desc=AP}, B=@n69), B=DotProduct(A=@n69, B=@n69))
  n71 = Multiply(A=OneMinus(None=Saturate(None=@n66)), B=OneMinus(None=Saturate(None=@n70)))
  n72 = Power(
   Base=MaterialFunctionCall(Alpha=@n71) {materialFunction=&SmoothStep}.Result
   Exp=ScalarParameter() {defaultValue=1, parameterName=Downstream Power}
  ) {constExponent=5}
  n73 = Multiply(A=Saturate(None=Divide(A=@n62.Water Depth) {constB=100}), B=@n72)
  n74 = Constant4Vector() {constant=<rgba 0.1667 0 1 1>}
  n75 = LinearInterpolate(
   A=LinearInterpolate(A=Constant4Vector() {constant=<rgba 0.9 0.09 0.9 1>}.RGBA, B=Constant4Vector() {constant=<rgba 0.56 0.08 0.8 1>}.RGBA, Alpha=@n54)
   B=Constant4Vector() {constant=<rgba 0 0.6667 1 1>}.RGBA
   Alpha=@n60
  )
  n76 = Saturate(None=Multiply(A=@n73) {constB=5})
  n77 = StaticSwitchParameter(
   True=ComponentMask(None=LinearInterpolate(A=@n74.RGBA, B=@n75, Alpha=@n76)) {b=true, g=true, r=true}
   False=LinearInterpolate(A=VectorParameter() {parameterName=Color C, defaultValue=<rgba 0.0167 0 0.1 1>}.RGB, B=Power(Base=@n61), Alpha=@n73)
  ) {parameterName=UseSquidWaterMobile}
  n78 = Multiply(A=ScalarParameter() {defaultValue=1, parameterName=SquidWater Scattering Multiplier}, B=@n77)
  FunctionOutput(None=StaticSwitchParameter(True=LinearInterpolate(A=@n33, B=@n78, Alpha=@n72), False=@n33) {parameterName=UseSquidWater}) {outputName=Scattering}
  n79 = FunctionInput() {bUsePreviewValueAsDefault=true, inputName=Absorption, sortPriority=1, previewValue=<1 1 1 1>}
  n80 = ScalarParameter() {defaultValue=0.025, parameterName=SquidWater Absorption Multiplier}
  n81 = Multiply(A=@n80, B=@n79)
  FunctionOutput(None=StaticSwitchParameter(True=LinearInterpolate(A=@n79, B=@n81, Alpha=@n72), False=@n79) {parameterName=UseSquidWater}) {outputName=Absorption, sortPriority=1}
  n82 = FunctionInput() {bUsePreviewValueAsDefault=true, inputName=MobileEmissive, previewValue=<1 1 1 1>}
  FunctionOutput(None=StaticSwitchParameter(True=LinearInterpolate(A=@n82, B=@n78, Alpha=@n72), False=@n82) {parameterName=UseSquidWater}) {outputName=MobileEmissive}
  n83 = Constant3Vector() {constant=<rgba 0 0 1 0>}
  n84 = Normalize(VectorInput=@n11)
  n85 = OneMinus(None=@n20)
  n86 = MaterialFunctionCall(Alpha=Saturate(None=Multiply(A=@n85) {constB=10})) {materialFunction=&SmoothStep}
  n87 = Multiply(A=@n86.Result, B=ScalarParameter() {defaultValue=-1, parameterName=Squid Normal Intensity 2})
  n88 = Normalize(VectorInput=AppendVector(A=Multiply(A=@n84, B=@n87), B=Constant() {r=1}))
  n89 = Saturate(None=Power(Base=@n20) {constExponent=10})
  n90 = LinearInterpolate(A=Multiply(A=@n88, B=Constant3Vector() {constant=<rgba -1 -1 1 0>}) {constB=-1}, B=@n88, Alpha=@n89)
  n91 = ScalarParameter() {defaultValue=0.3, parameterName=Squid Normal Intensity NoSpiral}
  n92 = ScalarParameter() {defaultValue=1, parameterName=Squid Normal Intensity Spiral}
  n93 = Power(Base=@n20)
  n94 = Add(A=@n52) {constB=-0.5}
  n95 = Multiply(
   A=Multiply(A=LinearInterpolate(A=@n91, B=@n92, Alpha=@n93), B=@n72)
   B=Multiply(A=ScalarParameter() {defaultValue=1, parameterName=Squid Normal Intensity}, B=Multiply(A=@n94) {constB=2})
  )
  n96 = AppendVector(A=Multiply(A=Multiply(A=@n86.Result, B=@n95), B=@n84), B=Constant() {r=1})
  n97 = Normalize(VectorInput=MaterialFunctionCall(BaseNormal=LinearInterpolate(A=@n83, B=@n90, Alpha=@n89), AdditionalNormal=Normalize(VectorInput=@n96)) {materialFunction=&BlendAngleCorrectedNormals}.Result)
  FunctionOutput(None=@n97) {outputName=Normal, sortPriority=1}
  n98 = FunctionInput() {bUsePreviewValueAsDefault=true, inputName=MobileOpacity, previewValue=<1 1 1 1>}
  n99 = DepthFade(FadeDistance=ScalarParameter() {defaultValue=30, parameterName=Mobile Depth Fade})
  FunctionOutput(None=StaticSwitchParameter(True=LinearInterpolate(A=@n98, B=@n99, Alpha=@n72), False=@n98) {parameterName=UseSquidWater}) {outputName=MobileOpacity}
  LinearInterpolate(A=Multiply(A=@n36) {constB=0.1}, B=@n36, Alpha=@n7.R)
  LinearInterpolate(A=@n35.RGB, B=@n34.RGB, Alpha=Saturate(None=@n20))
  n100 = LinearInterpolate(A=Constant3Vector() {constant=<rgba 0 0 1 0>}, B=@n97) {constAlpha=0.1}
  n101 = Saturate(None=DotProduct(A=CameraVectorWS(), B=Transform(None=@n100)))
  MaterialFunctionCall(Hue Shift Percentage=Multiply(A=OneMinus(None=@n101), B=ScalarParameter() {defaultValue=-0.2, parameterName=HueShift}), Texture=@n77) {materialFunction=&HueShift}
  MaterialFunctionCall(Alpha=@n57) {materialFunction=&SmoothStep}
  MaterialFunctionCall!(1=Constant(), 2=Constant() {r=0.1}, 3=Constant() {r=0.9}, 4=Constant() {r=1}, A=@n71) {materialFunction=/Engine/Functions/Engine_MaterialFunctions02/Lerp_Multiple_Float}
  MaterialFunctionCall() {materialFunction=&ComponentPivotLocation}
  n102 = AppendVector(A=WorldPosition().XY, B=ScalarParameter() {defaultValue=1500, parameterName=WWater Height})
  n103 = DistanceFieldGradient!(World Position=@n102)
  Multiply(A=Distance(A=ComponentMask(None=@n103) {g=true, r=true}, B=Constant())) {constB=1000}
  Multiply(A=@n52)
  n104 = Distance(A=@n12, B=Constant())
  n105 = Power(Base=Saturate(None=Divide(A=@n104) {constB=0.5})) {constExponent=5}
  Multiply(A=@n59, B=LinearInterpolate(Alpha=OneMinus(None=@n105)) {constA=0.5})
  n106 = LinearInterpolate(
   A=ScalarParameter() {defaultValue=200, parameterName=Depth Fade}
   B=ScalarParameter() {defaultValue=500, parameterName=Depth Fade 2}
   Alpha=@n54
  )
  n107 = MaterialFunctionCall(Alpha=DepthFade(FadeDistance=@n106)) {materialFunction=&SmoothStep}
  Multiply(A=Multiply(A=Multiply(A=@n107.Result, B=@n80)) {constB=111}, B=Constant3Vector() {constant=<rgba 1 1 1 0>})
  Multiply(A=Power(Base=@n19) {constExponent=0.05}) {constB=10}
  OneMinus(None=@n1)
  OneMinus(None=@n54)
  PixelNormalWS()
  Power(Base=@n28) {constExponent=5}
  Power(Base=@n75)
  Reroute() {desc=Water Depth}
  Reroute(None=@n10)
  n108 = Divide(A=@n9) {constB=11}
  Reroute(None=Multiply(A=Saturate(None=@n108), B=@n2))
  Reroute(None=@n51.Result)
  Reroute(None=@n38)
  Reroute(None=Divide(A=DistanceToNearestSurface!(World Position=@n102)) {constB=50})
  Reroute(None=@n47)
  Reroute(None=@n45.R)
  Reroute(None=@n46)
  n109 = MaterialFunctionCall(In=Power(Base=@n45.R), Contrast=Constant()) {materialFunction=&CheapContrast}
  Reroute(None=Multiply(A=LinearInterpolate(Alpha=@n109.Result) {constA=0.3, constB=0.5}, B=@n47) {constB=0.5})
  Reroute(None=Saturate(None=Divide(A=@n62.Water Depth, B=@n106) {constB=1}))
  n110 = Constant()
  Saturate(None=OneMinus(None=MaterialFunctionCall(In=@n48, Contrast=@n110) {materialFunction=&CheapContrast}.Result)) {desc=soft edges}
  Saturate(None=@n29)
  TextureCoordinate()
  VertexNormalWS()
 ]
}
